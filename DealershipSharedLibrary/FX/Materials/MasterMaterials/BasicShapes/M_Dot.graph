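# M_Dot is an unlit translucent master material for Niagara sprite and mesh particles that draws a soft round dot.
# Particle color RGB runs through MF_FX_EmissiveAdjustments for emissive.
# A scalar parameter blends between an exponential radial gradient and a contrast-powered UV distance falloff, and the result is multiplied by particle alpha.
# A static switch optionally applies depth fade to that opacity through MF_VCVFX_DepthFade.
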Material /DealershipSharedLibrary/FX/Materials/MasterMaterials/BasicShapes/M_Dot
graph {
 properties {
  bUsedWithNiagaraMeshParticles=true
  bUsedWithNiagaraSprites=true
  blendMode=BLEND_Translucent
  pixelDepthOffsetMode=PDOM_Legacy
  shadingModel=MSM_Unlit
  usedShadingModels=Unlit
 }
 nodes [
  MaterialFunctionCall(Emissive=ParticleColor().RGB) {materialFunction=&MF_FX_EmissiveAdjustments}
  n1 = ScalarParameter() {defaultValue=0.5, parameterName=Radius}
  n2 = ScalarParameter() {defaultValue=2.33, parameterName=Density}
  n3 = Add(A=TextureCoordinate(), B=Constant() {r=-0.5})
  n4 = DotProduct(A=@n3, B=@n3)
  n5 = Power(Base=OneMinus(None=SquareRoot(None=@n4)), Exp=ScalarParameter() {defaultValue=12, group=Dots, parameterName=Contrast})
  n6 = ScalarParameter() {parameterName="Use RadialGradient =0, DotMethod = 1"}
  n7 = Multiply(
   A=ParticleColor().A
   B=LinearInterpolate(
    A=MaterialFunctionCall(Radius=@n1, Density=@n2) {materialFunction=&RadialGradientExponential}.RadialGradientExponential
    B=Saturate(None=@n5)
    Alpha=Saturate(None=@n6)
   )
  )
  StaticSwitchParameter(
   True=MaterialFunctionCall(
    Opacity In=@n7
    DepthFade Distance=ScalarParameter() {defaultValue=100, group=Opacity, parameterName=DepthFade-Standard_Distance, sortPriority=0}
   ) {materialFunction=&MF_VCVFX_DepthFade}.Result
   False=@n7
  ) {parameterName=Use Deph Fade}
 ]
}
