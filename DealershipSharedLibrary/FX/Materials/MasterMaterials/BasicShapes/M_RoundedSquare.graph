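# Unlit masked master material for Niagara sprites that draws a texture-free rounded square tinted by particle color.
# SmoothStep ramps on the UVs and their inverse, set by Roundness, are multiplied across both axes, scaled by Size and thresholded by another SmoothStep.
# That shape is multiplied by particle alpha and passed through CheapContrast.
# Particle RGB runs through MF_FX_EmissiveAdjustments and an Emissive Gain multiplier.
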Material /DealershipSharedLibrary/FX/Materials/MasterMaterials/BasicShapes/M_RoundedSquare
graph {
 properties {bAutomaticallySetUsageInEditor=false, bUsedWithNiagaraSprites=true, blendMode=BLEND_Masked, shadingModel=MSM_Unlit, usedShadingModels=Unlit}
 nodes [
  n1 = ParticleColor()
  n2 = Constant()
  n3 = ScalarParameter() {defaultValue=0.2, parameterName=Roundness, sliderMax=0.5}
  n4 = TextureCoordinate()
  n5 = Multiply(A=SmoothStep(Min=@n2, Max=@n3, Value=@n4), B=SmoothStep(Min=@n2, Max=@n3, Value=OneMinus(None=@n4)))
  n6 = Multiply(
   A=Multiply(A=ComponentMask(None=@n5) {r=true}, B=ComponentMask(None=@n5) {g=true})
   B=ScalarParameter() {defaultValue=3, parameterName=Size, sliderMax=5, sliderMin=0.5}
  ) {constB=3}
  MaterialFunctionCall(In=Multiply(A=@n1.A, B=SmoothStep(Value=@n6) {constMin=0.5}), Contrast=ScalarParameter() {defaultValue=1, parameterName=Contrast}) {materialFunction=&CheapContrast}
  Multiply(
   A=MaterialFunctionCall(Emissive=@n1.RGB) {materialFunction=&MF_FX_EmissiveAdjustments}.Result
   B=ScalarParameter() {defaultValue=1, parameterName=Emissive Gain}
  )
 ]
}
