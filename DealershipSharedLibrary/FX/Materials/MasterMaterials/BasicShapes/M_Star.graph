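# Unlit translucent master material for Niagara sprite and mesh particles that draws a procedural cross star shape.
# The ProceduralCrossStar function builds the shape from tighten, softness and tip-fix controls, with side fade, softness and star power driven per particle through dynamic parameters.
# The powered shape is multiplied by particle alpha and passed through a depth-fade function.
# Particle color goes through the shared emissive adjustments function.
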Material /DealershipSharedLibrary/FX/Materials/MasterMaterials/BasicShapes/M_Star
graph {
 properties {
  bAutomaticallySetUsageInEditor=false
  bUsedWithNiagaraMeshParticles=true
  bUsedWithNiagaraSprites=true
  blendMode=BLEND_Translucent
  shadingModel=MSM_Unlit
  usedShadingModels=Unlit
 }
 nodes [
  n1 = ParticleColor()
  n2 = DynamicParameter() {defaultValue=<rgba 1 2 1 1>, paramNames [Star Power, Fade Sides, Softness, Param4]}
  n3 = Add(A=@n2.Fade Sides, B=ScalarParameter() {defaultValue=1, parameterName=Tighten})
  n4 = ScalarParameter() {parameterName=Fix If Tip Flickering}
  n5 = ScalarParameter() {defaultValue=1, parameterName=Fix Tip Width}
  n6 = Multiply(
   A=@n1.A
   B=Power(
    Base=MaterialFunctionCall!(Tighten=@n3, "Softness(0-1)"=@n2.Softness, "Fix_AntiFlick(Increase This If The Tip Flicking)"=@n4, Fix_TipWidth=@n5) {materialFunction=/Game/Effects/Fort_Effects/GameplayPulginsAssets/PulseLance/Slam/ProceduralCrossStar}.Result
    Exp=@n2.Star Power
   )
  )
  Reroute(None=@n6)
  Reroute(None=MaterialFunctionCall(Opacity In=@n6, DepthFade Distance=ScalarParameter() {defaultValue=100, parameterName=Depth Fade Distance}) {materialFunction=&MF_VCVFX_DepthFade}.Result)
  Reroute(None=MaterialFunctionCall(Emissive=@n1.RGB) {materialFunction=&MF_FX_EmissiveAdjustments}.Result)
 ]
}
